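# Physics asset for the TheHerald cape cosmetic, defining bodies and joints for the secondary motion of its dyn_cape bones.
# A large backpack_root capsule anchors an upper sphere hub carrying box leaves and sphere side branches, and a lower capsule spine.
# The spine forks into mid chains at its end and into side chains and small leaf chains at its second bone.
# Tight swing and twist limits keep the cape fairly stiff.
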PhysicsAsset /BRCosmetics/Accessories/FORT_Capes/Cape_F_MED_TheHerald/Meshes/F_MED_TheHerald_Cape_Physics
bodies {
 backpack_root [{shapeType=Capsule, length=38.25, radius=22.5, center=<0 20.3334 -33.5949>, rotation=<rot 0 0 6.2104>}]
 dyn_cape_down_1 [{shapeType=Capsule, length=9.4318, radius=3.5835, shapeName=dyn_cape_down_1_capsule, center=<9.0711 -0.1942 0.1118>}]
 dyn_cape_down_2 [{shapeType=Capsule, length=11.75, radius=4.25, shapeName=dyn_cape_down_2_capsule, center=<7.4281 -0.5861 -0.7901>}]
 dyn_cape_down_3 [{shapeType=Capsule, length=2.5, radius=2.5, shapeName=dyn_cape_down_3_capsule, center=<8.0881 -0.6867 -0.1175>}]
 dyn_cape_down_4 [{shapeType=Capsule, length=1.5, radius=2.5, shapeName=dyn_cape_down_4_capsule, center=<6.7248 0.8984 0.2592>}]
 dyn_cape_down_leaf_A_1 [{shapeType=Capsule, length=2, radius=2, shapeName=dyn_cape_down_leaf_A_1_capsule, center=<1.3818 0 0>}]
 dyn_cape_down_leaf_A_2 [{shapeType=Capsule, length=2, radius=2, shapeName=dyn_cape_down_leaf_A_2_capsule, center=<1.3818 0 0>}]
 dyn_cape_down_leaf_B_1 [{shapeType=Capsule, length=2, radius=2, shapeName=dyn_cape_down_leaf_B_1_capsule, center=<1.3818 0 0>}]
 dyn_cape_down_leaf_B_2 [{shapeType=Capsule, length=2, radius=2, shapeName=dyn_cape_down_leaf_B_2_capsule, center=<1.3818 0 0>}]
 dyn_cape_down_leaf_C_1 [{shapeType=Capsule, length=2, radius=2, shapeName=dyn_cape_down_leaf_C_1_capsule, center=<1.3818 0 0>}]
 dyn_cape_down_leaf_C_2 [{shapeType=Capsule, length=2, radius=2, shapeName=dyn_cape_down_leaf_C_2_capsule, center=<1.3818 0 0>}]
 dyn_cape_down_leaf_D_1 [{shapeType=Capsule, length=2, radius=2, shapeName=dyn_cape_down_leaf_D_1_capsule, center=<1.3818 0 0>}]
 dyn_cape_down_leaf_D_2 [{shapeType=Capsule, length=2, radius=2, shapeName=dyn_cape_down_leaf_D_2_capsule, center=<1.3818 0 0>}]
 dyn_cape_down_mid_1_l [{
   shapeType=Capsule
   length=2.5
   radius=2.5
   shapeName=dyn_cape_down_mid_1_l_capsule
   center=<5.1309 0.3111 0.178>
   rotation=<rot 0 -78.1042 90>
  }]
 dyn_cape_down_mid_1_r [{
   shapeType=Capsule
   length=2.5
   radius=2.5
   shapeName=dyn_cape_down_mid_1_r_capsule
   center=<7.4235 0.5316 -0.3002>
   rotation=<rot 0 93.2659 90>
  }]
 dyn_cape_down_mid_2_l [{
   shapeType=Capsule
   length=2.5
   radius=2.5
   shapeName=dyn_cape_down_mid_2_l_capsule
   center=<5.1629 0.6212 -0.1738>
   rotation=<rot 0 90.8546 90>
  }]
 dyn_cape_down_mid_2_r [{shapeType=Capsule, length=2.5, radius=2.5, shapeName=dyn_cape_down_mid_2_r_capsule, center=<5.8174 1.0437 0.0471>, rotation=<rot 0 -90 90>}]
 dyn_cape_down_mid_3_l [{
   shapeType=Capsule
   length=2.5
   radius=2.5
   shapeName=dyn_cape_down_mid_3_l_capsule
   center=<4.3947 0.2484 -0.0749>
   rotation=<rot 0 83.1326 90>
  }]
 dyn_cape_down_mid_3_r [{
   shapeType=Capsule
   length=2.5
   radius=2.5
   shapeName=dyn_cape_down_mid_3_r_capsule
   center=<3.672 0.1498 -0.2505>
   rotation=<rot 0 87.3906 90>
  }]
 dyn_cape_down_mid_4_l [{
   shapeType=Capsule
   length=2.5
   radius=2.5
   shapeName=dyn_cape_down_mid_4_l_capsule
   center=<2.5053 -0.4391 -0.3011>
   rotation=<rot 0 -98.2928 90>
  }]
 dyn_cape_down_mid_4_r [{
   shapeType=Capsule
   length=2.5
   radius=2.5
   shapeName=dyn_cape_down_mid_4_r_capsule
   center=<6.1228 -0.6534 0.1266>
   rotation=<rot 0 84.9076 90>
  }]
 dyn_cape_down_side_1_l [{shapeType=Capsule, length=2.5, radius=2.5, shapeName=dyn_cape_down_side_1_l_capsule, center=<6.8568 2.3112 -1.4839>}]
 dyn_cape_down_side_1_r [{shapeType=Capsule, length=2.5, radius=2.5, shapeName=dyn_cape_down_side_1_r_capsule, center=<7.0955 1.6952 -0.197>}]
 dyn_cape_down_side_2_l [{shapeType=Capsule, length=1.5, radius=2.5, shapeName=dyn_cape_down_side_2_l_capsule, center=<5.2134 2.5524 -1.9101>}]
 dyn_cape_down_side_2_r [{shapeType=Capsule, length=1.5, radius=2.5, shapeName=dyn_cape_down_side_2_r_capsule, center=<1 0 0>}]
 dyn_cape_down_side_3_l [{shapeType=Capsule, length=2.5, radius=2.5, shapeName=dyn_cape_down_side_3_l_capsule, center=<1 0 0>}]
 dyn_cape_down_side_3_r [{shapeType=Capsule, length=2.5, radius=2.5, shapeName=dyn_cape_down_side_3_r_capsule, center=<1 0 0>}]
 dyn_cape_down_side_4_l [{shapeType=Capsule, length=2.5, radius=2.5, shapeName=dyn_cape_down_side_4_l_capsule, center=<1 0 0>}]
 dyn_cape_down_side_4_r [{shapeType=Capsule, length=2.5, radius=2.5, shapeName=dyn_cape_down_side_4_r_capsule, center=<1 0 0>}]
 dyn_cape_down_side_5_l [{shapeType=Capsule, length=2.5, radius=2.5, shapeName=dyn_cape_down_side_5_l_capsule, center=<1 0 0>}]
 dyn_cape_down_side_5_r [{shapeType=Capsule, length=2.5, radius=2.5, shapeName=dyn_cape_down_side_5_r_capsule, center=<1 0 0>}]
 dyn_cape_down_side_6_l [{shapeType=Capsule, length=2.5, radius=2.5, shapeName=dyn_cape_down_side_6_l_capsule, center=<1 0 0>}]
 dyn_cape_up [{shapeType=Sphere, radius=7.1076, shapeName=dyn_cape_up_sphere}]
 dyn_cape_up_leaf_A_l [{
   shapeType=Box
   extentX=1
   extentY=1
   extentZ=7.8599
   shapeName=dyn_cape_up_leaf_A_l_box
   center=<4.3633 -0.378 0.155>
   rotation=<rot 0 -91.5305 -89.7265>
  }]
 dyn_cape_up_leaf_A_r [{
   shapeType=Box
   extentX=1
   extentY=1
   extentZ=6.4162
   shapeName=dyn_cape_up_leaf_A_r_box
   center=<4.6088 -0.2995 -0.3925>
   rotation=<rot 0 91.0582 -88.5414>
  }]
 dyn_cape_up_leaf_B_l [{
   shapeType=Box
   extentX=1
   extentY=1
   extentZ=8.7715
   shapeName=dyn_cape_up_leaf_B_l_box
   center=<5.1199 -0.6267 0.0052>
   rotation=<rot 0 -89.2908 -83.56>
  }]
 dyn_cape_up_leaf_B_r [{
   shapeType=Box
   extentX=1
   extentY=1
   extentZ=8.7179
   shapeName=dyn_cape_up_leaf_B_r_box
   center=<5.1709 -0.7578 -0.0279>
   rotation=<rot 0 92.0273 -84.1855>
  }]
 dyn_cape_up_leaf_C_l [{
   shapeType=Box
   extentX=1
   extentY=1
   extentZ=9.571
   shapeName=dyn_cape_up_leaf_C_l_box
   center=<4.694 -0.5014 0.0665>
   rotation=<rot 0 -93.1781 -86.5508>
  }]
 dyn_cape_up_leaf_C_r [{
   shapeType=Box
   extentX=1
   extentY=1
   extentZ=9.4055
   shapeName=dyn_cape_up_leaf_C_r_box
   center=<4.7815 -0.4125 -0.0288>
   rotation=<rot 0 86.8047 -86.3712>
  }]
 dyn_cape_up_leaf_D_1_l [{
   shapeType=Box
   extentX=1
   extentY=1
   extentZ=8.4333
   shapeName=dyn_cape_up_leaf_D_1_l_box
   center=<4.271 -0.7136 -0.2105>
   rotation=<rot 0 -89.4833 -87.2033>
  }]
 dyn_cape_up_leaf_D_1_r [{
   shapeType=Box
   extentX=1
   extentY=1
   extentZ=8.3883
   shapeName=dyn_cape_up_leaf_D_1_r_box
   center=<4.2458 -0.7001 0.2368>
   rotation=<rot 0 90.3556 -87.2221>
  }]
 dyn_cape_up_leaf_D_2_l [{
   shapeType=Box
   extentX=1
   extentY=1
   extentZ=6.8936
   shapeName=dyn_cape_up_leaf_D_2_l_box
   center=<4.6056 -0.6441 0.0734>
   rotation=<rot 0 95.0598 -87.8076>
  }]
 dyn_cape_up_leaf_D_2_r [{
   shapeType=Box
   extentX=1
   extentY=1
   extentZ=7.4679
   shapeName=dyn_cape_up_leaf_D_2_r_box
   center=<5.2527 -0.664 -0.2153>
   rotation=<rot 0 -89.4219 -89.7829>
  }]
 dyn_cape_up_leaf_E_l [{
   shapeType=Box
   extentX=1
   extentY=1
   extentZ=4.5871
   shapeName=dyn_cape_up_leaf_E_l_box
   center=<3.1231 -0.0748 0.3898>
   rotation=<rot 0 80.3823 -81.8193>
  }]
 dyn_cape_up_leaf_E_r [{
   shapeType=Box
   extentX=1
   extentY=1
   extentZ=6.9823
   shapeName=dyn_cape_up_leaf_E_r_box
   center=<1.9158 0.1618 -0.4119>
   rotation=<rot 0 -103.1735 -78.8988>
  }]
 dyn_cape_up_leaf_F_l [{
   shapeType=Box
   extentX=1
   extentY=1
   extentZ=6.885
   shapeName=dyn_cape_up_leaf_F_l_box
   center=<4.698 -0.7467 -0.0039>
   rotation=<rot 0 95.7893 -88.6527>
  }]
 dyn_cape_up_leaf_G_l [{
   shapeType=Box
   extentX=1
   extentY=1
   extentZ=4.4991
   shapeName=dyn_cape_up_leaf_G_l_box
   center=<2.077 -0.0736 0.1249>
   rotation=<rot 0 -88.1325 -88.621>
  }]
 dyn_cape_up_leaf_G_r [{
   shapeType=Box
   extentX=1
   extentY=1
   extentZ=5.2402
   shapeName=dyn_cape_up_leaf_G_r_box
   center=<2.8812 0.0455 -0.0561>
   rotation=<rot 0 89.1471 -89.0476>
  }]
 dyn_cape_up_leaf_mid_1_l [{
   shapeType=Box
   extentX=1
   extentY=1
   extentZ=10.3
   shapeName=dyn_cape_up_leaf_mid_1_l_box
   center=<4.826 -0.8905 -0.0243>
   rotation=<rot 0 -88.7496 -89.9039>
  }]
 dyn_cape_up_leaf_mid_2_l [{
   shapeType=Box
   extentX=1
   extentY=1
   extentZ=7.0615
   shapeName=dyn_cape_up_leaf_mid_2_l_box
   center=<3.9864 -0.715 0.0703>
   rotation=<rot 0 -90.6907 -87.4441>
  }]
 dyn_cape_up_leaf_side_1_l [{shapeType=Sphere, radius=3, shapeName=dyn_cape_up_leaf_side_1_l_sphere}]
 dyn_cape_up_leaf_side_1_r [{shapeType=Sphere, radius=3, shapeName=dyn_cape_up_leaf_side_1_r_sphere}]
 dyn_cape_up_leaf_side_2_l [{shapeType=Sphere, radius=2, shapeName=dyn_cape_up_leaf_side_2_l_sphere}]
 dyn_cape_up_leaf_side_2_r [{shapeType=Sphere, radius=2, shapeName=dyn_cape_up_leaf_side_2_r_sphere}]
 dyn_cape_up_leaf_side_3_l [{shapeType=Sphere, radius=2, shapeName=dyn_cape_up_leaf_side_3_l_sphere}]
 dyn_cape_up_leaf_side_3_r [{shapeType=Sphere, radius=2, shapeName=dyn_cape_up_leaf_side_3_r_sphere}]
 dyn_cape_up_leaf_side_4_l [{shapeType=Sphere, radius=2, shapeName=dyn_cape_up_leaf_side_4_l_sphere}]
 dyn_cape_up_leaf_side_4_r [{shapeType=Sphere, radius=2, shapeName=dyn_cape_up_leaf_side_4_r_sphere}]
}
constraints [
 {bone1Name=dyn_cape_up, bone2Name=backpack_root, swing1LimitDegrees=2, swing2LimitDegrees=2, twistLimitDegrees=2}
 {bone1Name=dyn_cape_up_leaf_A_r, bone2Name=dyn_cape_up, swing1LimitDegrees=5, swing2LimitDegrees=5, twistLimitDegrees=3}
 {bone1Name=dyn_cape_up_leaf_B_l, bone2Name=dyn_cape_up, swing1LimitDegrees=5, swing2LimitDegrees=5, twistLimitDegrees=3}
 {bone1Name=dyn_cape_up_leaf_A_l, bone2Name=dyn_cape_up_leaf_B_l, swing1LimitDegrees=5, swing2LimitDegrees=5, twistLimitDegrees=3}
 {bone1Name=dyn_cape_up_leaf_B_r, bone2Name=dyn_cape_up, swing1LimitDegrees=5, swing2LimitDegrees=5, twistLimitDegrees=3}
 {bone1Name=dyn_cape_up_leaf_C_l, bone2Name=dyn_cape_up, swing1LimitDegrees=5, swing2LimitDegrees=5, twistLimitDegrees=3}
 {bone1Name=dyn_cape_up_leaf_C_r, bone2Name=dyn_cape_up, swing1LimitDegrees=5, swing2LimitDegrees=5, twistLimitDegrees=3}
 {bone1Name=dyn_cape_up_leaf_D_1_l, bone2Name=dyn_cape_up, swing1LimitDegrees=5, swing2LimitDegrees=5, twistLimitDegrees=3}
 {bone1Name=dyn_cape_up_leaf_D_2_l, bone2Name=dyn_cape_up_leaf_D_1_l, swing1LimitDegrees=5, swing2LimitDegrees=5, twistLimitDegrees=3}
 {bone1Name=dyn_cape_up_leaf_D_1_r, bone2Name=dyn_cape_up, swing1LimitDegrees=5, swing2LimitDegrees=5, twistLimitDegrees=3}
 {bone1Name=dyn_cape_up_leaf_D_2_r, bone2Name=dyn_cape_up_leaf_D_1_r, swing1LimitDegrees=5, swing2LimitDegrees=5, twistLimitDegrees=3}
 {bone1Name=dyn_cape_up_leaf_E_l, bone2Name=dyn_cape_up, swing1LimitDegrees=5, swing2LimitDegrees=5, twistLimitDegrees=3}
 {bone1Name=dyn_cape_up_leaf_E_r, bone2Name=dyn_cape_up, swing1LimitDegrees=5, swing2LimitDegrees=5, twistLimitDegrees=3}
 {bone1Name=dyn_cape_up_leaf_F_l, bone2Name=dyn_cape_up, swing1LimitDegrees=5, swing2LimitDegrees=5, twistLimitDegrees=3}
 {bone1Name=dyn_cape_up_leaf_G_l, bone2Name=dyn_cape_up, swing1LimitDegrees=5, swing2LimitDegrees=5, twistLimitDegrees=3}
 {bone1Name=dyn_cape_up_leaf_G_r, bone2Name=dyn_cape_up, swing1LimitDegrees=5, swing2LimitDegrees=5, twistLimitDegrees=3}
 {bone1Name=dyn_cape_up_leaf_mid_1_l, bone2Name=dyn_cape_up, swing1LimitDegrees=5, swing2LimitDegrees=5, twistLimitDegrees=3}
 {bone1Name=dyn_cape_up_leaf_mid_2_l, bone2Name=dyn_cape_up, swing1LimitDegrees=5, swing2LimitDegrees=5, twistLimitDegrees=3}
 {bone1Name=dyn_cape_up_leaf_side_1_l, bone2Name=dyn_cape_up, swing1LimitDegrees=2, swing2LimitDegrees=10, twistLimitDegrees=3}
 {bone1Name=dyn_cape_up_leaf_side_2_l, bone2Name=dyn_cape_up_leaf_side_1_l, swing1LimitDegrees=2, swing2LimitDegrees=10, twistLimitDegrees=3}
 {bone1Name=dyn_cape_up_leaf_side_3_l, bone2Name=dyn_cape_up_leaf_side_1_l, swing1LimitDegrees=2, swing2LimitDegrees=10, twistLimitDegrees=3}
 {bone1Name=dyn_cape_up_leaf_side_4_l, bone2Name=dyn_cape_up_leaf_side_3_l, swing1LimitDegrees=2, swing2LimitDegrees=10, twistLimitDegrees=3}
 {bone1Name=dyn_cape_up_leaf_side_1_r, bone2Name=dyn_cape_up, swing1LimitDegrees=2, swing2LimitDegrees=10, twistLimitDegrees=3}
 {bone1Name=dyn_cape_up_leaf_side_2_r, bone2Name=dyn_cape_up_leaf_side_1_r, swing1LimitDegrees=2, swing2LimitDegrees=10, twistLimitDegrees=3}
 {bone1Name=dyn_cape_up_leaf_side_3_r, bone2Name=dyn_cape_up_leaf_side_1_r, swing1LimitDegrees=2, swing2LimitDegrees=10, twistLimitDegrees=3}
 {bone1Name=dyn_cape_up_leaf_side_4_r, bone2Name=dyn_cape_up_leaf_side_3_r, swing1LimitDegrees=2, swing2LimitDegrees=10, twistLimitDegrees=3}
 {bone1Name=dyn_cape_down_side_5_r, bone2Name=dyn_cape_down_side_4_r, swing1LimitDegrees=3, swing2LimitDegrees=5, twistLimitDegrees=2}
 {bone1Name=dyn_cape_down_1, bone2Name=backpack_root, swing1LimitDegrees=10, swing2LimitDegrees=10, twistLimitDegrees=5}
 {bone1Name=dyn_cape_down_2, bone2Name=dyn_cape_down_1, swing1LimitDegrees=3, swing2LimitDegrees=3, twistLimitDegrees=5}
 {bone1Name=dyn_cape_down_3, bone2Name=dyn_cape_down_2, swing1LimitDegrees=0.5, swing2LimitDegrees=2, twistLimitDegrees=3}
 {bone1Name=dyn_cape_down_4, bone2Name=dyn_cape_down_3, swing1LimitDegrees=0.6, swing2LimitDegrees=2, twistLimitDegrees=3}
 {bone1Name=dyn_cape_down_mid_1_l, bone2Name=dyn_cape_down_4, swing1LimitDegrees=0.6, swing2LimitDegrees=2, twistLimitDegrees=3}
 {bone1Name=dyn_cape_down_mid_2_l, bone2Name=dyn_cape_down_mid_1_l, swing1LimitDegrees=0.6, swing2LimitDegrees=2, twistLimitDegrees=2}
 {bone1Name=dyn_cape_down_mid_3_l, bone2Name=dyn_cape_down_mid_2_l, swing1LimitDegrees=0.6, swing2LimitDegrees=2, twistLimitDegrees=2}
 {bone1Name=dyn_cape_down_mid_4_l, bone2Name=dyn_cape_down_mid_3_l, swing1LimitDegrees=0.6, swing2LimitDegrees=2, twistLimitDegrees=2}
 {bone1Name=dyn_cape_down_mid_1_r, bone2Name=dyn_cape_down_4, swing1LimitDegrees=0.6, swing2LimitDegrees=2, twistLimitDegrees=3}
 {bone1Name=dyn_cape_down_mid_2_r, bone2Name=dyn_cape_down_mid_1_r, swing1LimitDegrees=0.6, swing2LimitDegrees=2, twistLimitDegrees=2}
 {bone1Name=dyn_cape_down_mid_3_r, bone2Name=dyn_cape_down_mid_2_r, swing1LimitDegrees=0.6, swing2LimitDegrees=2, twistLimitDegrees=2}
 {bone1Name=dyn_cape_down_mid_4_r, bone2Name=dyn_cape_down_mid_3_r, swing1LimitDegrees=0.6, swing2LimitDegrees=2, twistLimitDegrees=2}
 {bone1Name=dyn_cape_down_leaf_A_1, bone2Name=dyn_cape_down_2, swing1LimitDegrees=10, swing2LimitDegrees=10, twistLimitDegrees=5}
 {bone1Name=dyn_cape_down_leaf_A_2, bone2Name=dyn_cape_down_leaf_A_1, swing1LimitDegrees=10, swing2LimitDegrees=10, twistLimitDegrees=5}
 {bone1Name=dyn_cape_down_leaf_B_1, bone2Name=dyn_cape_down_2, swing1LimitDegrees=10, swing2LimitDegrees=10, twistLimitDegrees=5}
 {bone1Name=dyn_cape_down_leaf_B_2, bone2Name=dyn_cape_down_leaf_B_1, swing1LimitDegrees=10, swing2LimitDegrees=10, twistLimitDegrees=5}
 {bone1Name=dyn_cape_down_leaf_C_1, bone2Name=dyn_cape_down_2, swing1LimitDegrees=10, swing2LimitDegrees=10, twistLimitDegrees=5}
 {bone1Name=dyn_cape_down_leaf_C_2, bone2Name=dyn_cape_down_leaf_C_1, swing1LimitDegrees=10, swing2LimitDegrees=10, twistLimitDegrees=5}
 {bone1Name=dyn_cape_down_leaf_D_1, bone2Name=dyn_cape_down_2, swing1LimitDegrees=10, swing2LimitDegrees=10, twistLimitDegrees=5}
 {bone1Name=dyn_cape_down_leaf_D_2, bone2Name=dyn_cape_down_leaf_D_1, swing1LimitDegrees=10, swing2LimitDegrees=10, twistLimitDegrees=5}
 {bone1Name=dyn_cape_down_side_1_l, bone2Name=dyn_cape_down_2, swing1LimitDegrees=3, swing2LimitDegrees=3, twistLimitDegrees=3}
 {bone1Name=dyn_cape_down_side_2_l, bone2Name=dyn_cape_down_side_1_l, swing1LimitDegrees=3, swing2LimitDegrees=3, twistLimitDegrees=3}
 {bone1Name=dyn_cape_down_side_3_l, bone2Name=dyn_cape_down_side_2_l, swing1LimitDegrees=3, swing2LimitDegrees=5, twistLimitDegrees=3}
 {bone1Name=dyn_cape_down_side_4_l, bone2Name=dyn_cape_down_side_3_l, swing1LimitDegrees=3, swing2LimitDegrees=5, twistLimitDegrees=2}
 {bone1Name=dyn_cape_down_side_5_l, bone2Name=dyn_cape_down_side_4_l, swing1LimitDegrees=3, swing2LimitDegrees=5, twistLimitDegrees=2}
 {bone1Name=dyn_cape_down_side_6_l, bone2Name=dyn_cape_down_side_5_l, swing1LimitDegrees=3, swing2LimitDegrees=5, twistLimitDegrees=2}
 {bone1Name=dyn_cape_down_side_1_r, bone2Name=dyn_cape_down_2, swing1LimitDegrees=3, swing2LimitDegrees=3, twistLimitDegrees=3}
 {bone1Name=dyn_cape_down_side_2_r, bone2Name=dyn_cape_down_side_1_r, swing1LimitDegrees=3, swing2LimitDegrees=3, twistLimitDegrees=3}
 {bone1Name=dyn_cape_down_side_3_r, bone2Name=dyn_cape_down_side_2_r, swing1LimitDegrees=3, swing2LimitDegrees=5, twistLimitDegrees=3}
 {bone1Name=dyn_cape_down_side_4_r, bone2Name=dyn_cape_down_side_3_r, swing1LimitDegrees=3, swing2LimitDegrees=5, twistLimitDegrees=2}
]
